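# Looping stereo MetaSound preset of MSS_Glider_Thrust_Loop that supplies the thrust and movement audio for the Season 30 Umbrella glider.
# It only overrides inputs: forward, backward and sideways share one thrust loop and ten MoveSweet variations, while the base thrust layer uses an idle loop and no one-shots.
# The parent graph receives Fort.Glider direction values.
# Glider attenuation and concurrency apply, and a glider-open envelope-follower control bus modulates volume.
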MetaSoundSource /BRCosmetics/Sounds/Gliders/UmbrellaS30/MSS_Glider_Umbrella_S30_Thrust
preset_parent=/FortMetasound/Cosmetics/Glider/MSS_Glider_Thrust_Loop
inputs [
 Backward.Loop.Sounds {type=WaveAsset:Array}
 Backward.Sounds {type=WaveAsset:Array}
 Backward.Volume {type=Float}
 Forward.Loop.Sounds {type=WaveAsset:Array}
 Forward.Sounds {type=WaveAsset:Array}
 Forward.Volume {type=Float}
 Glider.AnyDirection {type=Float}
 Glider.Backwards {type=Float}
 Glider.Forwards {type=Float}
 Glider.Sideways {type=Float}
 Sideways.Loop.Sounds {type=WaveAsset:Array}
 Sideways.Sounds {type=WaveAsset:Array}
 Sideways.Volume {type=Float}
 Thrust.Loop.Sounds {type=WaveAsset:Array}
 Thrust.OneShotVolume {type=Float}
 Thrust.Sounds {type=WaveAsset:Array}
 Thrust.Volume {type=Float}
 UE.Source.OnPlay {type=Trigger}
]
interfaces [Fort.Glider:1.0, UE.OutputFormat.Stereo:1.0, UE.Source:1.1]
outputs [
 "UE.OutputFormat.Stereo.Audio:0" {type=Audio}
 "UE.OutputFormat.Stereo.Audio:1" {type=Audio}
]
pages [Default {input_defaults {
   Backward.Volume=1
   Forward.Volume=1
   Glider.AnyDirection=0
   Glider.Backwards=0
   Glider.Forwards=0
   Glider.Sideways=0
   Sideways.Volume=1
   Thrust.OneShotVolume=0.6
   Thrust.Volume=1
   UE.Source.OnPlay=null
   Backward.Loop.Sounds [/BRCosmetics/Sounds/Gliders/Season30_Umbrella/Umbrella_S30_ThrustLoop_01]
   Backward.Sounds [
    /BRCosmetics/Sounds/Gliders/Season30_Umbrella/Umbrella_S30_MoveSweet_01
    /BRCosmetics/Sounds/Gliders/Season30_Umbrella/Umbrella_S30_MoveSweet_02
    /BRCosmetics/Sounds/Gliders/Season30_Umbrella/Umbrella_S30_MoveSweet_03
    /BRCosmetics/Sounds/Gliders/Season30_Umbrella/Umbrella_S30_MoveSweet_04
    /BRCosmetics/Sounds/Gliders/Season30_Umbrella/Umbrella_S30_MoveSweet_05
    /BRCosmetics/Sounds/Gliders/Season30_Umbrella/Umbrella_S30_MoveSweet_06
    /BRCosmetics/Sounds/Gliders/Season30_Umbrella/Umbrella_S30_MoveSweet_07
    /BRCosmetics/Sounds/Gliders/Season30_Umbrella/Umbrella_S30_MoveSweet_08
    /BRCosmetics/Sounds/Gliders/Season30_Umbrella/Umbrella_S30_MoveSweet_09
    /BRCosmetics/Sounds/Gliders/Season30_Umbrella/Umbrella_S30_MoveSweet_10
   ]
   Forward.Loop.Sounds [&Umbrella_S30_ThrustLoop_01]
   Forward.Sounds [
    &Umbrella_S30_MoveSweet_01
    &Umbrella_S30_MoveSweet_02
    &Umbrella_S30_MoveSweet_03
    &Umbrella_S30_MoveSweet_04
    &Umbrella_S30_MoveSweet_05
    &Umbrella_S30_MoveSweet_06
    &Umbrella_S30_MoveSweet_07
    &Umbrella_S30_MoveSweet_08
    &Umbrella_S30_MoveSweet_09
    &Umbrella_S30_MoveSweet_10
   ]
   Sideways.Loop.Sounds [&Umbrella_S30_ThrustLoop_01]
   Sideways.Sounds [
    &Umbrella_S30_MoveSweet_01
    &Umbrella_S30_MoveSweet_02
    &Umbrella_S30_MoveSweet_03
    &Umbrella_S30_MoveSweet_04
    &Umbrella_S30_MoveSweet_05
    &Umbrella_S30_MoveSweet_06
    &Umbrella_S30_MoveSweet_07
    &Umbrella_S30_MoveSweet_08
    &Umbrella_S30_MoveSweet_09
    &Umbrella_S30_MoveSweet_10
   ]
   Thrust.Loop.Sounds [/BRCosmetics/Sounds/Gliders/Season30_Umbrella/Umbrella_S30_IdleLoop_01]
   Thrust.Sounds []
  }}]
properties {
 assetImportData=&self:AssetImportData
 attenuationSettings=&Fort_Player_Glider_Thrust_Athena_Att
 audioPropertiesSheet=&APS_Cosmetic_Glider_Loop
 bLooping=true
 duration=10000
 maxDistance=12500
 numChannels=2
 outputFormat=Stereo
 sampleRate=48000
 soundClassObject=&Fort_Gliders_Land
 volume=0.35
 concurrencySet [&Athena_Glider_Thrust_Conc]
 modulationSettings {
  volumeRouting=Union
  volumeModulationDestination {bEnableModulation=true, modulators [/Game/Sounds/Modulation/EnvelopeFollowers/Volume/CB_Glider_Open_EnvFollower]}
 }
}
